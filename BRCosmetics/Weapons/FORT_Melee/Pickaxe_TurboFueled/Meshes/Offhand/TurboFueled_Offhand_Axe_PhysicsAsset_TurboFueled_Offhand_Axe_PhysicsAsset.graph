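# Physics asset for the offhand axe mesh of the TurboFueled pickaxe, a Battle Royale cosmetic melee weapon.
# It holds a single body on the Melee_TwoHanded_Handle bone, shaped as one offset, rotated capsule that serves as a simplified volume for the weapon.
# It defines no constraints, so it has no jointed or independently linked parts.
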PhysicsAsset /BRCosmetics/Weapons/FORT_Melee/Pickaxe_TurboFueled/Meshes/Offhand/TurboFueled_Offhand_Axe_PhysicsAsset_TurboFueled_Offhand_Axe_PhysicsAsset
bodies {Melee_TwoHanded_Handle [{shapeType=Capsule, length=21.1734, radius=12.0954, center=<-2.4597 7.6859 -1.2681>, rotation=<rot 0 6.3118 -88.3888>}]}
constraints []
